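# Translucent lit material for KelpLinen ripple effects on Niagara mesh particles, rendering a glowing, water-like surface with time-of-day color.
# Parallax scanline layers, noise clouds projected on two virtual planes, sky-lit sparkles and a world-space slurp mask build the emissive, over foam and scrolling normals.
# UV-based front and back zones shape the layers, while particle dynamic parameters drive dissolves.
# Depth fade, particle alpha and a location-fade function control opacity.
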Material /Clyde/Environments/KelpLinen/TechArt/M_KelpLinen_Ripple
graph {
 properties {
  bAutomaticallySetUsageInEditor=false
  bHasPixelAnimation=true
  bUseMaterialAttributes=true
  bUsedWithNiagaraMeshParticles=true
  blendMode=BLEND_Translucent
  translucencyLightingMode=TLM_SurfacePerPixelLighting
  usedShadingModels=Default Lit
 }
 nodes [
  Constant3Vector() {constant=<rgba 2 0.2958 1.7728 1>}
  Constant3Vector() {constant=<rgba 2 0.2958 1.7728 1>}
  n1 = Constant3Vector() {constant=<rgba 0 0 1 0>}
  n2 = NamedRerouteUsage()
  n3 = NamedRerouteUsage()
  n4 = Multiply(A=ComponentMask(None=@n3) {g=true, r=true}, B=ScalarParameter() {parameterName=Cloudy Texture Scroll Speed Distortion})
  n5 = ScalarParameter() {parameterName=Cloudy Offset Foreground Min}
  n6 = Add(A=ComponentMask(None=@n3) {b=true}, B=@n5)
  n7 = Constant2Vector() {g=1, r=1}
  n8 = ScalarParameter() {defaultValue=2000, parameterName=Cloudy Tiling Distortion}
  n9 = TextureSample(
   UVs=MaterialFunctionCall(Plane Normal Axis=@n1, Plane X Axis=@n2, Plane Center=AppendVector(A=@n4, B=@n6), UV Scale=Multiply(A=@n7, B=@n8)) {materialFunction=&VirtualPlaneCoordinates}.UVs
   Tex=TextureObject() {samplerType=SAMPLERTYPE_Normal, texture=&/T_noise_02_nml}
  )
  ConstantBiasScale(None=@n9.R) {bias=-0.5, scale=2}
  LinearInterpolate(A=@n5, B=ScalarParameter() {defaultValue=100, parameterName=Cloudy Offset Foreground Max}, Alpha=NamedRerouteUsage()) {constB=200}
  n10 = ScalarParameter() {defaultValue=-300, parameterName=Cloudy Offset Background Min}
  LinearInterpolate(A=@n10, B=ScalarParameter() {defaultValue=-100, parameterName=Cloudy Offset Background Max}, Alpha=NamedRerouteUsage()) {constB=200}
  n11 = Multiply(A=ComponentMask(None=NamedRerouteUsage()) {g=true, r=true}, B=ScalarParameter() {parameterName=Sparkle Hue Shift Scroll Speed})
  n12 = WorldPosition()
  n13 = TextureSample(UVs=Divide(A=Subtract(A=@n11, B=@n12.XY), B=ScalarParameter() {defaultValue=100, parameterName=Sparkle Hue Shift Tiling}) {constB=1000}) {
   samplerSource=SSM_Wrap_WorldGroupSettings
   samplerType=SAMPLERTYPE_Grayscale
   texture=/Game/TimeOfDay/TODM_Sequence/Materials/Sky/Textures/LowResBlurredNoise_Radish_LF_Root
  }
  n14 = ScalarParameter() {defaultValue=75, group=PostRVTSnow, parameterName=Morning/Evening-Multi}
  n15 = Multiply(
   A=VectorParameter() {parameterName=Sparkle_Color, defaultValue=<rgba 15 2.8125 14.5377 1>}.RGB
   B=MaterialFunctionCall(
    Value_Morning=@n14
    Value_Day=ScalarParameter() {defaultValue=150, group=PostRVTSnow, parameterName=Day_Multi}
    Value_Evening=@n14
    Value_Night=ScalarParameter() {defaultValue=450, group=PostRVTSnow, parameterName=Night_Emissive_Multi}
   ) {materialFunction=&Scalar_Time_of_Day}.Result
  )
  MaterialFunctionCall(Hue Shift Percentage=LinearInterpolate(Alpha=@n13.R) {constB=0.3}, Texture=@n15) {materialFunction=&HueShift}
  n16 = ScalarParameter() {defaultValue=5, parameterName=Inner Depth Variance Front}
  n17 = ScalarParameter() {defaultValue=4, parameterName=Inner Depth Variance Back}
  n18 = Power(Base=NamedRerouteUsage()) {constExponent=0.8}
  n19 = LinearInterpolate(A=@n16, B=@n17, Alpha=@n18) {constB=0}
  n20 = ComponentMask(None=NamedRerouteUsage()) {g=true, r=true}
  n21 = WorldPosition()
  n22 = Subtract(A=Multiply(A=@n20, B=ScalarParameter() {defaultValue=0.5, parameterName=Inner Noise Scroll Speed Front}), B=@n21.XY)
  n23 = ScalarParameter() {defaultValue=8000, parameterName=Inner Noise Tiling Front}
  n24 = Multiply(A=@n20, B=ScalarParameter() {defaultValue=0.2, parameterName=Inner Noise Scroll Speed Back})
  n25 = TextureSample(UVs=Divide(A=Subtract(A=@n24, B=@n21.XY), B=ScalarParameter() {defaultValue=50000, parameterName=Inner Noise Tiling Back}) {constB=1000}) {samplerType=SAMPLERTYPE_Grayscale, texture=&/LowResBlurredNoise}
  n26 = LinearInterpolate(
   A=TextureSample(UVs=Divide(A=@n22, B=@n23) {constB=1000}) {samplerType=SAMPLERTYPE_Grayscale, texture=&LowResBlurredNoise}.R
   B=@n25.R
   Alpha=@n18
  )
  n27 = LinearInterpolate(
   A=ScalarParameter() {defaultValue=1, parameterName=Inner Depth Offset Front}
   B=ScalarParameter() {defaultValue=1, parameterName=Inner Depth Offset Back}
   Alpha=@n18
  )
  n28 = Add(A=Multiply(A=@n26, B=@n19), B=@n27)
  n29 = ScalarParameter() {defaultValue=0.3, parameterName=Inner Depth Variance Spread Front, sliderMax=1}
  n30 = Abs(None=ConstantBiasScale(None=NamedRerouteUsage()) {bias=-0.5, scale=2})
  n31 = Multiply(A=@n16, B=LinearInterpolate(A=OneMinus(None=@n29), Alpha=Power(Base=@n30)) {constA=1}) {constB=1.5}
  n32 = ScalarParameter() {defaultValue=1, parameterName=Inner Depth Variance Spread Back, sliderMax=1}
  n33 = NamedRerouteUsage()
  n34 = Multiply(A=@n17, B=LinearInterpolate(A=OneMinus(None=@n32), Alpha=Power(Base=@n33)))
  n35 = Add(A=Multiply(A=@n26, B=LinearInterpolate(A=@n31, B=@n34, Alpha=@n18) {constB=0}), B=@n27)
  Max(A=@n28, B=@n35)
  n36 = ScalarParameter() {defaultValue=2000, parameterName=Cloudy Accent Tiling X}
  Multiply(A=Constant2Vector() {g=1, r=1}, B=@n36)
  n37 = Constant3Vector() {constant=<rgba 0 1 0 0>}
  NamedRerouteDeclaration(None=Normalize(VectorInput=Transform(None=@n37) {transformSourceType=TRANSFORMSOURCE_Instance})) {name=Forward Vector WS, nodeColor=<rgba 1 0 0.2109 1>}
  n38 = DotProduct(A=VertexNormalWS(), B=CameraVectorWS())
  n39 = InverseLinearInterpolate(
   A=ScalarParameter() {parameterName=Lines Dot Min}
   B=ScalarParameter() {defaultValue=0.3, parameterName=Lines Dot Max}
   Value=Saturate(None=NamedRerouteDeclaration(None=@n38) {name=CameraDotVert, nodeColor=<rgba 0 0.4375 1 1>})
  ) {constA=0.4}
  n40 = TextureCoordinate()
  n41 = Multiply(
   A=Power(Base=NamedRerouteUsage(), Exp=ScalarParameter() {defaultValue=0.8, parameterName=Line Tiling Power Front}) {constExponent=0.8}
   B=ScalarParameter() {defaultValue=1, parameterName=Line Tiling Front}
  )
  n42 = Multiply(
   A=Power(Base=NamedRerouteUsage(), Exp=ScalarParameter() {defaultValue=0.7, parameterName=Line Tiling Power Back}) {constExponent=0.6}
   B=ScalarParameter() {defaultValue=2, parameterName=Line Tiling Back}
  )
  n43 = NamedRerouteUsage()
  n44 = NamedRerouteUsage()
  n45 = Panner(
   Coordinate=MaterialFunctionCall(X=ComponentMask(None=@n40) {r=true}, Y=Add(A=@n41, B=@n42)) {materialFunction=&MakeFloat2}.Result
   Time=Length(None=Multiply(A=@n43, B=@n44)) {is=class MaterialExpressionLength}
   Speed=MaterialFunctionCall(X=Constant(), Y=ScalarParameter() {defaultValue=0, parameterName=Inner Texture Scroll Speed}) {materialFunction=&MakeFloat2}.Result
  ) {speedX=1}
  n46 = NamedRerouteUsage()
  n47 = LinearInterpolate(
   A=ScalarParameter() {defaultValue=1, parameterName=Refraction Strength Front}
   B=ScalarParameter() {defaultValue=1, parameterName=Refraction Strength Back}
   Alpha=NamedRerouteUsage()
  )
  n48 = NamedRerouteUsage()
  n49 = NamedRerouteUsage()
  n50 = Subtract(
   A=Multiply(A=ComponentMask(None=@n49) {g=true, r=true}, B=ScalarParameter() {defaultValue=0.5, parameterName=Line Noise Speed Back})
   B=WorldPosition().XY
  )
  n51 = ScalarParameter() {defaultValue=100000, parameterName=Line Noise Tiling Back}
  n52 = ConstantBiasScale(None=TextureSample(UVs=Divide(A=@n50, B=@n51) {constB=1000}) {samplerType=SAMPLERTYPE_Grayscale, texture=&LowResBlurredNoise}.R) {bias=-0.5, scale=2}
  n53 = Saturate(None=InverseLinearInterpolate(
    A=ScalarParameter() {defaultValue=0.05, parameterName=Line Noise Fade Y Start}
    B=ScalarParameter() {defaultValue=0.6, parameterName=Line Noise Fade Y End}
    Value=NamedRerouteUsage()
   ))
  n54 = LinearInterpolate(
   A=ScalarParameter() {defaultValue=0.2, parameterName=Line Noise Offset Strength Back}
   Alpha=Power(Base=OneMinus(None=@n53), Exp=ScalarParameter() {defaultValue=2, parameterName=Line Noise Fade Y Power})
  ) {constB=0}
  n55 = Add(
   A=Multiply(A=ComponentMask(None=@n46) {g=true, r=true}, B=LinearInterpolate(B=@n47, Alpha=@n48) {constB=0}) {constB=0.2}
   B=MaterialFunctionCall(Y=Multiply(A=@n52, B=@n54)) {materialFunction=&MakeFloat2}.Result
  )
  n56 = Add(A=BumpOffset(Coordinate=@n45, Height=Multiply(A=@n28) {constB=-1}), B=@n55) {constB=0}
  n57 = TextureObject() {texture=&/T_Scanlines_Offset_Horiz}
  n58 = LinearInterpolate(
   A=ScalarParameter() {defaultValue=3, parameterName=Line Power}
   B=ScalarParameter() {defaultValue=3, parameterName=Line Dissolve Power}
   Alpha=NamedRerouteUsage()
  ) {constA=1}
  n59 = ScalarParameter() {defaultValue=0.8, parameterName=Line Depth Fade Max}
  n60 = ScalarParameter() {defaultValue=0.1, parameterName=Line Depth Fade Min}
  n61 = InverseLinearInterpolate(A=@n59, B=@n60, Value=NamedRerouteUsage())
  n62 = Multiply(
   A=Power(Base=TextureSample(UVs=@n56, Tex=@n57) {texture=&T_Scanlines_Offset_Horiz}.RGB, Exp=@n58)
   B=OneMinus(None=Saturate(None=@n61))
  )
  n63 = Add(A=BumpOffset(Coordinate=@n45, Height=Multiply(A=@n35) {constB=-1}), B=@n55) {constB=0}
  n64 = InverseLinearInterpolate(A=@n59, B=@n60, Value=NamedRerouteUsage())
  n65 = Multiply(
   A=Power(Base=TextureSample(UVs=@n63, Tex=@n57) {texture=&T_Scanlines_Offset_Horiz}.RGB, Exp=@n58)
   B=OneMinus(None=Saturate(None=@n64))
  )
  n66 = ScalarParameter() {defaultValue=0.07, parameterName=Close Line Distortion Strength}
  n67 = NamedRerouteUsage()
  n68 = LinearInterpolate(A=Multiply(A=@n66) {constB=-1}, Alpha=@n67) {constB=0}
  n69 = NamedRerouteUsage()
  n70 = Subtract(
   A=Multiply(A=ComponentMask(None=@n69) {g=true, r=true}, B=ScalarParameter() {defaultValue=0.5, parameterName=Close Line Noise Speed})
   B=WorldPosition().XY
  )
  n71 = ScalarParameter() {defaultValue=5000, parameterName=Close Line Noise Tiling}
  n72 = LinearInterpolate(
   A=ScalarParameter() {defaultValue=0.25, parameterName=Close Line Noise Min}
   Alpha=TextureSample(UVs=Divide(A=@n70, B=@n71) {constB=1000}) {samplerType=SAMPLERTYPE_Grayscale, texture=&LowResBlurredNoise}.R
  )
  n73 = AppendVector(A=ComponentMask(None=@n63) {r=true}, B=Add(A=ComponentMask(None=@n63) {g=true}, B=Multiply(A=@n72, B=@n68)))
  n74 = Add(B=Power(
    Base=TextureSample(UVs=@n73, Tex=@n57) {texture=&T_Scanlines_Offset_Horiz}.R
    Exp=ScalarParameter() {defaultValue=6, parameterName=Close Line Power}
   ))
  n75 = OneMinus(None=NamedRerouteUsage())
  n76 = Multiply(A=Power(Base=Saturate(None=@n39)) {constExponent=1}, B=Add(A=Add(A=@n62, B=@n65), B=Multiply(A=@n74, B=@n75)))
  n77 = Saturate(None=InverseLinearInterpolate(
    A=ScalarParameter() {defaultValue=0.05, parameterName=Line Fade Y Start}
    B=ScalarParameter() {defaultValue=0.9, parameterName=Line Fade Y End}
    Value=NamedRerouteUsage()
   ))
  n78 = SphereMask(
   A=Divide(A=LocalPosition().XYZ, B=VectorParameter() {parameterName=Line Fade Sphere Scale, defaultValue=<rgba 1 0.7 1 1>}.RGB)
   B=VectorParameter() {parameterName=Line Fade Sphere Position, defaultValue=<rgba 0 -35000 0 1>}.RGB
   Radius=ScalarParameter() {defaultValue=50000, parameterName=Line Fade Sphere Radius}
   Hardness=ScalarParameter() {parameterName=Line Fade Sphere Hardness}
  )
  n79 = Multiply(
   A=Power(Base=OneMinus(None=@n77), Exp=ScalarParameter() {defaultValue=2, parameterName=Line Fade Y Power})
   B=Power(Base=OneMinus(None=@n78)) {constExponent=1}
  )
  NamedRerouteDeclaration(None=Multiply(A=Multiply(A=@n76, B=@n79), B=NamedRerouteUsage())) {name=Line Opacity, nodeColor=<rgba 1 0.6328 0 1>}
  n80 = NamedRerouteUsage()
  n81 = ScalarParameter() {defaultValue=0.6, parameterName=Back Opacity Fade Line Power}
  n82 = InverseLinearInterpolate(
   A=ScalarParameter() {parameterName=Back Opacity Fade Start}
   B=ScalarParameter() {defaultValue=0.1, parameterName=Back Opacity Fade End}
   Value=NamedRerouteUsage()
  )
  n83 = Power(Base=OneMinus(None=Saturate(None=@n82)), Exp=ScalarParameter() {defaultValue=2, parameterName=Back Opacity Fade Power})
  n84 = OneMinus(None=NamedRerouteUsage())
  n85 = ScalarParameter() {defaultValue=500, parameterName=Depth Fade Distance Near}
  n86 = ScalarParameter() {defaultValue=250, parameterName=Depth Fade Distance Far}
  n87 = NamedRerouteUsage()
  n88 = Multiply(
   A=Max(A=Power(Base=@n80, Exp=@n81) {constExponent=0.6}, B=Multiply(A=@n83, B=@n84))
   B=DepthFade(FadeDistance=LinearInterpolate(A=@n85, B=@n86, Alpha=@n87)) {fadeDistanceDefault=500}
  )
  n89 = ParticleColor()
  NamedRerouteDeclaration(None=Multiply(
    A=Multiply(A=@n88, B=@n89.A)
    B=MaterialFunctionCall() {materialFunction=/ClydeCosmetics/Environments/KelpLinen/MF_KelpLinen_LocationFade}.Output
   )) {name=Opacity Out, nodeColor=<rgba 1 0.6328 0 1>}
  n90 = WorldPosition()
  n91 = CameraPositionWS()
  NamedRerouteDeclaration(None=Length(None=Subtract(A=@n90.XYZ, B=@n91))) {name=Camera Distance, nodeColor=<rgba 0.3594 0 1 1>}
  n92 = VectorParameter() {parameterName=Absorption Color Day Close, defaultValue=<rgba 0.1042 0.8059 1 1>}
  n93 = VectorParameter() {parameterName=Absorption Color Day, defaultValue=<rgba 0.0695 0.3717 0.8938 1>}
  n94 = NamedRerouteUsage()
  NamedRerouteDeclaration(None=MaterialFunctionCall(
    Color_Morning=VectorParameter() {parameterName=Absorption Color Morning, defaultValue=<rgba 0.1136 0.3 0.6729 1>}.RGB
    Color_Day=LinearInterpolate(A=@n92.RGB, B=@n93.RGB, Alpha=@n94)
    Color_Evening=VectorParameter() {parameterName=Absorption Color Evening, defaultValue=<rgba 0.1252 0.4553 0.6604 1>}.RGB
    Color_Night=VectorParameter() {parameterName=Absorption Color Night, defaultValue=<rgba 0.0342 0.1785 0.5979 1>}.RGB
   ) {materialFunction=&Color_Time_of_Day}.Result) {name=Absorption Color, nodeColor=<rgba 1 0.4483 0 1>}
  n95 = Add(
   A=Multiply(A=NamedRerouteUsage(), B=VectorParameter() {parameterName=Lines Color, defaultValue=<rgba 1.3123 14.9403 15 1>}.RGB)
   B=NamedRerouteUsage()
  )
  n96 = NamedRerouteUsage()
  n97 = ComponentMask(None=@n96) {g=true, r=true}
  n98 = ScalarParameter() {defaultValue=0.2, parameterName=Cloudy Texture Scroll Speed Foreground}
  n99 = ComponentMask(None=@n96) {b=true}
  n100 = MaterialFunctionCall(
   Plane Normal Axis=@n1
   Plane X Axis=@n2
   Plane Center=AppendVector(A=Multiply(A=@n97, B=@n98), B=Add(A=@n99, B=@n5))
   UV Scale=Multiply(A=Constant2Vector() {g=1, r=1}, B=ScalarParameter() {defaultValue=20000, parameterName=Cloudy Tiling Foreground})
  ) {materialFunction=&VirtualPlaneCoordinates}
  n101 = NamedRerouteUsage()
  n102 = ComponentMask(None=@n101) {g=true, r=true}
  n103 = ScalarParameter() {parameterName=Cloudy Texture Scroll Speed Background}
  n104 = ComponentMask(None=@n101) {b=true}
  n105 = MaterialFunctionCall(
   Plane Normal Axis=@n1
   Plane X Axis=@n2
   Plane Center=AppendVector(A=Multiply(A=@n102, B=@n103), B=Add(A=@n104, B=@n10))
   UV Scale=Multiply(A=Constant2Vector() {g=1, r=1}, B=ScalarParameter() {defaultValue=14000, parameterName=Cloudy Tiling Background})
  ) {materialFunction=&VirtualPlaneCoordinates}
  n106 = ScalarParameter() {defaultValue=0.2, parameterName=Cloudy Low Clip Foreground}
  n107 = ScalarParameter() {defaultValue=1, parameterName=Cloudy High Clip Foreground}
  n108 = Length(None=Subtract(A=@n100.World Position, B=CameraPositionWS()))
  n109 = NamedRerouteUsage()
  n110 = ScalarParameter() {defaultValue=50000, parameterName=Cloudy Depth Fade}
  n111 = Divide(A=Abs(None=Subtract(A=@n108, B=@n109)), B=@n110)
  n112 = Multiply(
   A=LinearInterpolate(B=ScalarParameter() {defaultValue=50, parameterName=Cloudy Refraction Strength}, Alpha=Saturate(None=@n111)) {constB=0}
   B=ComponentMask(None=NamedRerouteUsage()) {g=true, r=true}
  ) {constB=0.2}
  n113 = Multiply(A=ComponentMask(None=@n9.RGB) {g=true, r=true}, B=ScalarParameter() {defaultValue=0.05, parameterName=Cloudy Distortion Strength})
  n114 = TextureObject() {texture=&/T_Noise_001}
  n115 = TextureSample(UVs=Add(A=Add(A=@n100.UVs, B=@n112) {constB=0}, B=@n113), Tex=@n114) {texture=/Clyde/ExtraDependencies/AtmosphericCloudNoise01}
  n116 = OneMinus(None=@n115.G)
  n117 = MaterialFunctionCall(X=@n116, Y=@n116, Z=@n116) {materialFunction=&MakeFloat3}
  n118 = Power(
   Base=Saturate(None=InverseLinearInterpolate(A=@n106, B=@n107, Value=@n117.Result))
   Exp=ScalarParameter() {defaultValue=2, parameterName=Cloudy Foreground Power}
  ) {constExponent=1}
  n119 = OneMinus(None=@n118)
  n120 = Subtract(A=LinearInterpolate(A=@n100.World Position, B=@n105.World Position, Alpha=@n119), B=CameraPositionWS())
  n121 = Abs(None=Subtract(A=Length(None=@n120), B=NamedRerouteUsage()))
  n122 = Divide(A=@n121, B=ScalarParameter() {defaultValue=5000, parameterName=Cloudy Depth Strength}) {constB=100}
  n123 = Saturate(None=Divide(A=@n121, B=@n110) {constB=100})
  n124 = Multiply(
   A=Power(Base=Saturate(None=@n122), Exp=ScalarParameter() {defaultValue=0.8, parameterName=Cloudy Depth Power})
   B=OneMinus(None=Power(Base=@n123) {constExponent=1})
  )
  n125 = Divide(A=@n121, B=ScalarParameter() {defaultValue=4000, parameterName=Cloudy Depth Absorption})
  n126 = LinearInterpolate(
   B=NamedRerouteUsage()
   Alpha=Power(Base=Saturate(None=@n125), Exp=ScalarParameter() {defaultValue=2, parameterName=Cloudy Depth Absorption Power}) {constExponent=0.8}
  ) {constA=1, constB=0}
  n127 = ScalarParameter() {parameterName=Cloudy Low Clip Background}
  n128 = ScalarParameter() {defaultValue=0.25, parameterName=Cloudy High Clip Background}
  n129 = TextureSample(UVs=Add(A=Add(A=@n105.UVs, B=@n112), B=@n113), Tex=@n114) {texture=&AtmosphericCloudNoise01}
  n130 = Power(
   Base=Saturate(None=InverseLinearInterpolate(A=@n127, B=@n128, Value=@n129.RGB))
   Exp=ScalarParameter() {defaultValue=1, parameterName=Cloudy Background Power}
  ) {constExponent=1}
  n131 = Add(
   A=Multiply(A=@n118, B=VectorParameter() {parameterName=Cloudy Color Foreground, defaultValue=<rgba 10 0.5247 9.8478 1>}.RGB)
   B=Multiply(A=Multiply(A=@n130, B=@n119), B=VectorParameter() {parameterName=Cloudy Color Background, defaultValue=<rgba 0.7581 0.0927 1 1>}.RGB)
  ) {constB=0}
  n132 = ComponentMask(None=CameraPositionWS()) {b=true}
  n133 = WorldPosition()
  n134 = Floor(None=Saturate(None=Subtract(A=@n132, B=@n133.Z)))
  n135 = InverseLinearInterpolate(
   A=ScalarParameter() {parameterName=Cloudy Opacity Start}
   B=ScalarParameter() {defaultValue=0.6, parameterName=Cloudy Opacity End}
   Value=OneMinus(None=NamedRerouteUsage())
  ) {constA=0.94, constB=0.96}
  n136 = Multiply(A=Multiply(A=Multiply(A=@n124, B=@n126), B=Multiply(A=@n131, B=@n134)), B=Power(Base=Saturate(None=@n135)))
  n137 = NamedRerouteUsage()
  n138 = OneMinus(None=Saturate(None=NamedRerouteUsage()))
  n139 = Multiply(A=Multiply(A=Multiply(A=@n136, B=@n137), B=NamedRerouteUsage()), B=@n138)
  n140 = NamedRerouteUsage()
  n141 = VectorParameter() {parameterName=Slurp Color, defaultValue=<rgba 0 0.9833 1 1>}
  n142 = OneMinus(None=NamedRerouteUsage())
  n143 = Add(A=Add(A=Add(A=@n95, B=@n139) {constB=0}, B=Multiply(A=@n140, B=@n141.RGB)), B=Multiply(A=Min(A=@n138, B=@n142), B=NamedRerouteUsage()))
  n144 = ParticleColor()
  NamedRerouteDeclaration(None=Divide(A=Multiply(A=@n143, B=@n144.RGB), B=EyeAdaptation().EyeAdaptation)) {name=Emissive Out, nodeColor=<rgba 1 0.6328 0 1>}
  n145 = Multiply(A=Time(), B=ScalarParameter() {defaultValue=0.2, parameterName=Time Test Speed})
  NamedRerouteDeclaration(None=ConstantBiasScale(None=Sine(None=@n145))) {name=Temp Time, nodeColor=<rgba 1 0.7031 0 1>}
  n146 = ComponentMask(None=@n139) {r=true}
  n147 = ComponentMask(None=@n139) {g=true}
  NamedRerouteDeclaration(None=Max(A=Max(A=@n146, B=@n147), B=ComponentMask(None=@n139) {b=true})) {name=Plane Opacity, nodeColor=<rgba 1 0.6328 0 1>}
  n148 = NamedRerouteUsage()
  n149 = Subtract(
   A=Multiply(A=ComponentMask(None=@n148) {g=true, r=true}, B=ScalarParameter() {defaultValue=0.1, parameterName=Sparkle Scroll Speed})
   B=WorldPosition().XY
  )
  n150 = ScalarParameter() {defaultValue=250, parameterName=Sprakle Tiling}
  n151 = ConstantBiasScale(None=@n25.R) {bias=-0.5, scale=2}
  n152 = ScalarParameter() {defaultValue=1, parameterName=Sparkle Distortion Strength}
  NamedRerouteDeclaration(None=Add(A=Divide(A=@n149, B=@n150) {constB=1000}, B=Multiply(A=@n151, B=@n152))) {name=SparkleCoords, nodeColor=<rgba 1 0 0.2812 1>}
  n153 = NamedRerouteUsage()
  n154 = InverseLinearInterpolate(B=@n153, Value=Power(Base=@n116, Exp=ScalarParameter() {defaultValue=1, parameterName=Cloudy Dissolve Power}))
  NamedRerouteDeclaration(None=Multiply(A=Saturate(None=@n154), B=OneMinus(None=@n153))) {name=Pink Dissolve, nodeColor=<rgba 1 0.0703 0 1>}
  n155 = Add(A=@n19, B=@n27)
  NamedRerouteDeclaration(None=Divide(A=NamedRerouteDeclaration(None=@n28) {name=Maximum Line Depth 0, nodeColor=<rgba 0 1 0.7656 1>}, B=@n155)) {name=0-1 Line Depth 0, nodeColor=<rgba 0 1 0.7656 1>}
  n156 = DynamicParameter() {defaultValue=<rgba 0 0 0 1>, paramNames [PinkDissolve, Lines Dissolve, Blue Opacity, Param4]}
  NamedRerouteDeclaration(None=@n156.PinkDissolve) {name=Pink Controls, nodeColor=<rgba 0.8516 1 0 1>}
  NamedRerouteDeclaration(None=@n156.Lines Dissolve) {name=Line Controls, nodeColor=<rgba 0.8516 1 0 1>}
  NamedRerouteDeclaration(None=@n156.Blue Opacity) {name=Blue Controls, nodeColor=<rgba 0.8516 1 0 1>}
  n157 = NamedRerouteUsage()
  n158 = ScalarParameter() {defaultValue=3}
  n159 = Subtract(A=Multiply(A=@n20, B=ScalarParameter() {defaultValue=0.3, parameterName=Line Dissolve Noise Speed}), B=@n21.XY)
  n160 = ScalarParameter() {defaultValue=5000, parameterName=Line Dissolve Noise Tiling}
  n161 = InverseLinearInterpolate(
   B=@n157
   Value=TextureSample(UVs=Divide(A=@n159, B=@n160) {constB=1000}, Tex=TextureObject() {texture=&/CINE_T_SimpleNoise}) {samplerType=SAMPLERTYPE_LinearColor, texture=/FortWater/Textures/02_WaterFoam_Veroni_base_H}.R
  ) {constA=-0.01}
  n162 = Saturate(None=InverseLinearInterpolate(A=Power(Base=@n157, Exp=@n158), Value=Saturate(None=@n161)))
  n163 = ScalarParameter() {defaultValue=1, parameterName=Line Dissolve Noise Power}
  NamedRerouteDeclaration(None=Multiply(A=Power(Base=@n162, Exp=@n163), B=OneMinus(None=@n157))) {name=Line Dissolve, nodeColor=<rgba 1 0.0703 0 1>}
  n164 = ReflectionVectorWS(CustomWorldNormal=NamedRerouteUsage())
  n165 = Multiply(A=Multiply(A=SkyLightEnvMapSample(Direction=@n164), B=NamedRerouteUsage()), B=@n15)
  n166 = Constant()
  NamedRerouteDeclaration(None=Saturate(None=MaterialFunctionCall!(Default=@n165, Low/Mobile=@n166) {materialFunction=/Game/Effects/Fort_Effects/Materials/_Functions/MF_LowMobile_PathingSwitch}.Result)) {name=SparkleEmissive, nodeColor=<rgba 0 0.2734 1 1>}
  n167 = NamedRerouteUsage()
  n168 = TextureSample(UVs=@n167) {
   samplerSource=SSM_Wrap_WorldGroupSettings
   samplerType=SAMPLERTYPE_Normal
   texture=/Game/Environments/Asteria/Terrain/Textures/T_SnowSparkleCircles_N
  }
  n169 = ComponentMask(None=@n168.RGB) {b=true, g=true, r=true}
  n170 = Constant3Vector() {constant=<rgba 0 0 -0.9 1>}
  n171 = Transform(None=Normalize(VectorInput=Add(A=@n169, B=@n170)))
  n172 = Saturate(None=DotProduct(A=ReflectionVectorWS(CustomWorldNormal=@n171), B=CameraVectorWS()))
  n173 = TextureSample(UVs=@n167) {
   samplerSource=SSM_Wrap_WorldGroupSettings
   samplerType=SAMPLERTYPE_LinearColor
   texture=/Game/Environments/Asteria/Terrain/Textures/T_SnowSparkleCircles_M
  }
  n174 = ComponentMask(None=Saturate(None=Multiply(A=@n173.RGB))) {r=true}
  NamedRerouteDeclaration(None=Power(Base=Multiply(A=@n172, B=@n174), Exp=ScalarParameter() {defaultValue=4, group=PostRVTSnow, parameterName=SparkleMaskExponent})) {name=HighendSparkleMask, nodeColor=<rgba 0 0.1094 1 1>}
  NamedRerouteDeclaration(None=@n171) {name=sparklenormal, nodeColor=<rgba 1 0.1875 0 1>}
  n175 = MaterialFunctionCall() {materialFunction=&ObjectPivotPoint}
  n176 = ComponentMask(None=@n175.Mesh Particle Pivot Location) {g=true, r=true}
  n177 = Multiply(A=ParticleRandom(), B=Constant() {r=100000})
  NamedRerouteDeclaration(None=AppendVector(A=Add(A=@n176, B=@n177), B=ComponentMask(None=@n175.Mesh Particle Pivot Location) {b=true})) {name=Random Pivot Point, nodeColor=<rgba 0.3594 1 0 1>}
  n178 = OneMinus(None=ScalarParameter() {defaultValue=1, parameterName=Cloudy Accent Strength})
  n179 = NamedRerouteUsage()
  n180 = NamedRerouteUsage()
  n181 = ComponentMask(None=@n180) {g=true, r=true}
  n182 = ScalarParameter() {parameterName=Cloudy Accent Texture Scroll Speed}
  n183 = ComponentMask(None=@n180) {b=true}
  n184 = MaterialFunctionCall(
   Plane Normal Axis=@n1
   Plane X Axis=@n2
   Plane Center=AppendVector(A=Multiply(A=@n181, B=@n182), B=Add(A=@n183, B=@n5))
   UV Scale=MaterialFunctionCall(X=@n36, Y=ScalarParameter() {defaultValue=2000, parameterName=Cloudy Accent Tiling Y}) {materialFunction=&MakeFloat2}.Result
  ) {materialFunction=&VirtualPlaneCoordinates}
  NamedRerouteDeclaration(None=LinearInterpolate(A=LinearInterpolate(A=@n178, Alpha=@n179), Alpha=TextureSample(UVs=@n184.UVs, Tex=@n114).RGB)) {name=Cloudy Accent, nodeColor=<rgba 1 0 0.8865 1>}
  NamedRerouteDeclaration(None=MaterialFunctionCall(
    Fade Length=ScalarParameter() {defaultValue=10000, parameterName=Camera Effects Fade Length}
    Fade Offset=ScalarParameter() {defaultValue=1000, parameterName=Camera Effects Fade Offset}
   ) {materialFunction=&CameraDepthFade}.Result) {name=Camera Effects Fade, nodeColor=<rgba 0.3594 0 1 1>}
  n185 = ScalarParameter() {defaultValue=0.0025, group=Slurp, parameterName=Mask Range Min}
  n186 = ScalarParameter() {defaultValue=0.35, group=Slurp, parameterName=Mask Range Max}
  n187 = CollectionParameter() {collection=&FortniteMaterialParameters, parameterName=SlurpyCenter}
  n188 = ComponentMask(None=CollectionParameter() {collection=&FortniteMaterialParameters, parameterName=SlurpyRadius}) {b=true, g=true, r=true}
  n189 = CollectionParameter() {
   collection=&FortniteMaterialParameters
   parameterName=SlurpTextureMaskLocation
  }
  n190 = Add(A=WorldPosition().XYZ, B=OneMinus(None=ComponentMask(None=@n189) {b=true, g=true, r=true}))
  n191 = TextureObjectParameter() {parameterName=JuiceMask, samplerSource=SSM_Clamp_WorldGroupSettings, texture=/Clyde/Athena/Apollo/Water/Materials/Clyde_OilSpillV3}
  n192 = MaterialFunctionCall!(
   Export Float 4=StaticBoolParameter() {defaultValue=true, parameterName="Export Float 4 (SB)"}
   TextureObject=@n191
   TextureSize
   WorldPosition=@n190
   World Space Normal=@n190
   RotZAxis
   RotXAxis
   RotYAxis
   ProjectionTransitionContrast
  ) {materialFunction=/Clyde/Athena/Apollo/Water/Materials/MF_WorldAlignedTexture_NoWrap}
  n193 = Subtract(
   A=SphereMask(
    A=WorldPosition().XYZ
    B=ComponentMask(None=@n187) {b=true, g=true, r=true}
    Radius=Multiply(A=@n188) {constB=1.5}
    Hardness=Constant() {r=5}
   ) {hardnessPercent=0.1}
   B=OneMinus(None=ComponentMask(None=@n192.XYZ Texture) {r=true})
  )
  n194 = Power(
   Base=Saturate(None=InverseLinearInterpolate(A=@n185, B=@n186, Value=@n193))
   Exp=ScalarParameter() {defaultValue=0.3, group=Slurp, parameterName=Mask Power}
  )
  n195 = ScalarParameter() {defaultValue=1, group=Slurp, parameterName=Mask Multiplier}
  NamedRerouteDeclaration(None=Saturate(None=Multiply(A=@n194, B=@n195))) {name=SlurpMask, nodeColor=<rgba 0.0442 0.7756 1 1>}
  n196 = ComponentMask(None=NamedRerouteUsage()) {g=true, r=true}
  n197 = ScalarParameter() {defaultValue=0.3, parameterName=Foam Scroll Speed Front}
  n198 = Divide(A=Subtract(A=Multiply(A=@n196, B=@n197), B=WorldPosition().XY), B=ScalarParameter() {defaultValue=2500, parameterName=Foam Tiling}) {constB=1000}
  n199 = TextureObject() {samplerType=SAMPLERTYPE_Masks, texture=/FortWater/Textures/T_OceanFoam_New_Mask}
  n200 = Multiply(A=@n196, B=ScalarParameter() {defaultValue=0.8, parameterName=Foam Mask Scroll Speed Front})
  n201 = WorldPosition()
  n202 = TextureSample(
   UVs=Divide(A=Subtract(A=@n200, B=@n201.XY), B=ScalarParameter() {defaultValue=20000, parameterName=Foam Mask Tiling}) {constB=1000}
   Tex=@n199
  )
  n203 = Subtract(A=Add(A=TextureSample(UVs=@n198, Tex=@n199).R, B=@n202.B), B=@n202.G)
  n204 = InverseLinearInterpolate(
   A=ScalarParameter() {defaultValue=0.4, parameterName=Foam Min Cutoff Front}
   B=ScalarParameter() {defaultValue=0.9, parameterName=Foam Max Cutoff Front}
   Value=NamedRerouteUsage()
  )
  n205 = ScalarParameter() {defaultValue=0.9, parameterName=Foam Min Cutoff Back}
  n206 = ScalarParameter() {defaultValue=1, parameterName=Foam Max Cutoff Back}
  n207 = OneMinus(None=NamedRerouteUsage())
  n208 = Saturate(None=Multiply(A=Saturate(None=@n204), B=InverseLinearInterpolate(A=@n205, B=@n206, Value=@n207)))
  n209 = Power(Base=OneMinus(None=@n208), Exp=ScalarParameter() {defaultValue=1.5, parameterName=Foam Falloff Power Subtract})
  n210 = Saturate(None=Subtract(A=Subtract(A=@n203, B=@n209), B=ScalarParameter() {defaultValue=0.5, parameterName=Foam Strength Reduction}))
  n211 = Power(Base=@n208, Exp=ScalarParameter() {defaultValue=2, parameterName=Foam Falloff Power Multiply})
  n212 = NamedRerouteUsage()
  NamedRerouteDeclaration(None=Multiply(A=Multiply(A=@n210, B=@n211), B=OneMinus(None=@n212))) {name=Foam Opacity, nodeColor=<rgba 1 0.6328 0 1>}
  n213 = ScalarParameter() {defaultValue=0.3, parameterName=Slurp Min Cutoff}
  n214 = ScalarParameter() {defaultValue=0.7, parameterName=Slurp Max Cutoff}
  n215 = OneMinus(None=NamedRerouteUsage())
  n216 = Power(
   Base=Saturate(None=InverseLinearInterpolate(A=@n213, B=@n214, Value=@n215))
   Exp=ScalarParameter() {defaultValue=2, parameterName=Slurp Transition Power}
  ) {constExponent=0.7}
  n217 = Multiply(A=@n216, B=OneMinus(None=NamedRerouteUsage()))
  NamedRerouteDeclaration(None=Multiply(A=LinearInterpolate(A=@n217, Alpha=@n216), B=NamedRerouteUsage())) {name=Slurp Opacity, nodeColor=<rgba 1 0.6328 0 1>}
  NamedRerouteDeclaration(None=@n115.G) {name=Alpha Noise, nodeColor=<rgba 0 1 0.4609 1>}
  n218 = Constant3Vector() {constant=<rgba 0 0 1 0>}
  n219 = LinearInterpolate(
   A=@n218
   B=TextureSample(UVs=@n198, Tex=TextureObject() {samplerType=SAMPLERTYPE_Normal, texture=/FortWater/Textures/T_OceanFoam_New_N}).RGB
   Alpha=NamedRerouteUsage()
  )
  n220 = ScalarParameter() {defaultValue=0.2, parameterName=Foam Normal Strength}
  NamedRerouteDeclaration(None=Normalize(VectorInput=LinearInterpolate(A=@n218, B=@n219, Alpha=@n220))) {name=Foam Normal, nodeColor=<rgba 0.5 1 0 1>}
  NamedRerouteDeclaration(None=MaterialFunctionCall(
    Fade Length=ScalarParameter() {defaultValue=43732.9492, group=Distance Blending, parameterName=Skydive_Distance}
    Fade Offset=ScalarParameter() {defaultValue=500, group=Distance Blending, parameterName=Skydive_Offset}
   ) {materialFunction=&CameraDepthFade}.Result) {name=Skydive_Fade, nodeColor=<rgba 0.4297 1 0 1>}
  n221 = ScalarParameter() {defaultValue=0.05, parameterName=Zone Blend Location}
  n222 = ComponentMask(None=TextureCoordinate()) {g=true}
  NamedRerouteDeclaration(None=Saturate(None=InverseLinearInterpolate(B=@n221, Value=@n222) {constB=0.1})) {name=Front Zone, nodeColor=<rgba 0 0.1328 1 1>}
  NamedRerouteDeclaration(None=Saturate(None=InverseLinearInterpolate(A=@n221, Value=@n222) {constA=0.1})) {name=Back Zone, nodeColor=<rgba 0 0.1328 1 1>}
  n223 = Subtract(A=@n221, B=Min(A=@n221, B=ScalarParameter() {defaultValue=0.02, parameterName=Zone Blend Width Front, sliderMax=1}))
  n224 = Add(A=@n221, B=Min(A=OneMinus(None=@n221), B=ScalarParameter() {defaultValue=0.03, parameterName=Zone Blend Width Back}))
  NamedRerouteDeclaration(None=Saturate(None=InverseLinearInterpolate(A=@n223, B=@n224, Value=@n222) {constA=0.05, constB=0.1})) {name=Zone Blend, nodeColor=<rgba 0 0.1328 1 1>}
  n225 = NamedRerouteUsage()
  n226 = Subtract(
   A=Multiply(A=ComponentMask(None=@n225) {g=true, r=true}, B=ScalarParameter() {defaultValue=0.9, parameterName=Normal Scroll Speed})
   B=WorldPosition().XY
  )
  n227 = ScalarParameter() {defaultValue=250, parameterName=Normal Tiling}
  n228 = ComponentMask(None=TextureSample(UVs=Divide(A=@n226, B=@n227) {constB=3000}) {samplerType=SAMPLERTYPE_Normal, texture=&T_noise_02_nml}.RGB) {g=true, r=true}
  n229 = Divide(B=ScalarParameter() {defaultValue=0.02, parameterName=Normal Strength})
  NamedRerouteDeclaration(None=Normalize(VectorInput=AppendVector(A=@n228, B=@n229))) {name=Normals Out, nodeColor=<rgba 1 0.6328 0 1>}
  NamedRerouteDeclaration(None=Divide(A=NamedRerouteDeclaration(None=@n35) {name=Maximum Line Depth 1, nodeColor=<rgba 0 1 0.7656 1>}, B=@n155)) {name=0-1 Line Depth 1, nodeColor=<rgba 0 1 0.7656 1>}
  NamedRerouteUsage()
  NamedRerouteUsage()
  NamedRerouteUsage()
  Panner(Coordinate=MaterialFunctionCall!(Scale=Constant() {r=850}) {materialFunction=/Game/Athena/Apollo/Environments/Landscape/Materials/Biome/CommonMaterials/MF_SimpleWorldUV}.Result) {speedX=0.1}
  ScalarParameter() {parameterName=Temp Pink Controls}
  ScalarParameter() {parameterName=Temp Line Controls}
  ScalarParameter() {parameterName=Temp Blue Controls}
  SetMaterialAttributes!(
   MaterialAttributes
   Base Color=Multiply(A=NamedRerouteUsage(), B=VectorParameter() {parameterName=Foam Color, defaultValue=<rgba 0.2406 0.4407 0.4865 1>}.RGB)
   Metallic=Constant()
   Specular=ScalarParameter() {defaultValue=0.01, parameterName=Specular}
   Roughness=LinearInterpolate(A=ScalarParameter() {defaultValue=0.05, parameterName=Roughness}, Alpha=NamedRerouteUsage()) {constB=0.5}
   Normal=NamedRerouteUsage()
   Emissive Color=NamedRerouteUsage()
   Opacity=NamedRerouteUsage()
  ) {attributeSetTypes [
    #1
    #2
    #3
    #4
    #5
    #6
    #7
   ]}
  TextureObject() {samplerType=SAMPLERTYPE_Normal, texture=&T_noise_02_nml}
  TextureObject() {samplerType=SAMPLERTYPE_Masks, texture=/Game/Global/Textures/Default/Masks_and_Gradients/Swamp_Mask}
  n230 = ComponentMask(None=@n56) {g=true}
  n231 = Multiply(A=@n72, B=LinearInterpolate(A=@n66, Alpha=@n67) {constB=0})
  TextureSample(UVs=AppendVector(A=ComponentMask(None=@n56) {r=true}, B=Add(A=@n230, B=@n231)), Tex=@n57) {texture=&T_Scanlines_Offset_Horiz}
 ]
}
